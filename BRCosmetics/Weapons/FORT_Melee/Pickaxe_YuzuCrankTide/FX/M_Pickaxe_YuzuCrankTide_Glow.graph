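# Translucent unlit glow material for the YuzuCrankTide pickaxe cosmetic's effects, flagged for use on Niagara mesh particles.
# Its graph holds just two nodes: a Particle Color input and a linear grayscale sample of the pickaxe's edge glow mask texture.
# No links between these nodes or to the material outputs appear, so how color and mask combine is not shown.
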Material /BRCosmetics/Weapons/FORT_Melee/Pickaxe_YuzuCrankTide/FX/M_Pickaxe_YuzuCrankTide_Glow
graph {
 properties {
  bAutomaticallySetUsageInEditor=false
  bUsedWithNiagaraMeshParticles=true
  blendMode=BLEND_Translucent
  shadingModel=MSM_Unlit
  usedShadingModels=Unlit
 }
 nodes [
  ParticleColor()
  TextureSample() {
   samplerType=SAMPLERTYPE_LinearGrayscale
   texture=/BRCosmetics/Weapons/FORT_Melee/Pickaxe_YuzuCrankTide/Materials/T_YukuCrankTide_EdgeGlow_Mask
  }
 ]
}
