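# Translucent, unlit, two-sided glow material for YuzuCrankTide pickaxe Niagara sprites and ribbons.
# A panned normal map's red channel offsets the glow texture UVs, scaled by a distortion strength that defaults to zero.
# Glow red, optionally remapped through a curve-atlas LUT, is multiplied by particle color and divided by pre-exposure.
# Particle alpha times glow red forms opacity, with optional depth fade.
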Material /BRCosmetics/Weapons/FORT_Melee/Pickaxe_YuzuCrankTide/FX/M_Pickaxe_YuzuCrankTide_Glow_2
graph {
 properties {
  bEnableResponsiveAA=true
  bUsedWithNiagaraRibbons=true
  bUsedWithNiagaraSprites=true
  blendMode=BLEND_Translucent
  opacityMaskClipValue=0.03
  pixelDepthOffsetMode=PDOM_Legacy
  shadingModel=MSM_Unlit
  twoSided=true
  usedShadingModels=Unlit
 }
 nodes [
  n1 = ParticleColor()
  n2 = TextureCoordinate()
  n3 = MaterialFunctionCall(X=ScalarParameter() {defaultValue=1, parameterName=UV U Scale}, Y=ScalarParameter() {defaultValue=1, parameterName=UV V Scale}) {materialFunction=&MakeFloat2}
  n4 = ScalarParameter() {defaultValue=0.1, parameterName=UV Speed X}
  n5 = ScalarParameter() {parameterName=UV Speed Y}
  n6 = TextureSampleParameter2D(UVs=Panner(
    Coordinate=Multiply(A=@n2, B=@n3.Result)
    Speed=MaterialFunctionCall(X=@n4, Y=@n5) {materialFunction=&MakeFloat2}.Result
   )) {
   parameterName=UV Distort Texture
   samplerType=SAMPLERTYPE_Normal
   texture=&/T_YuzuCrankTide_Normal
  }
  n7 = AppendVector(A=@n6.R, B=@n6.R)
  n8 = ScalarParameter() {parameterName=UV Distort Strength}
  n9 = TextureSampleParameter2D(UVs=Add(A=TextureCoordinate(), B=Multiply(A=@n7, B=@n8))) {parameterName=TX_Glow, texture=&/T_Pickaxe_YuzuCrankTide_Glow}
  n10 = LinearInterpolate(A=@n9.R, B=@n9.R, Alpha=ScalarParameter() {parameterName=USE_TEXTURE_COLOR})
  n11 = CurveAtlasRowParameter(CurveTime=LinearInterpolate(
    A=ScalarParameter() {defaultValue=0.2, parameterName=LUT MIN}
    B=ScalarParameter() {defaultValue=1, parameterName=LUT MAX}
    Alpha=Saturate(None=@n10)
   )) {
   atlas=&/CA_Pickaxe_YuzuCrankTide
   curve=&/C_Pickaxe_YuzuCrankTide
   parameterName=BaseLUT
  }
  n12 = StaticBoolParameter() {parameterName=USE LUT}
  Divide(A=Multiply(A=@n1.RGB, B=StaticSwitch(True=@n11, False=@n10, Value=@n12)), B=ViewProperty() {property=MEVP_PreExposure}.Property)
  n13 = Multiply(A=@n1.A, B=LinearInterpolate(A=@n9.R, B=@n9.R, Alpha=ScalarParameter() {parameterName=USE_ALPHA}))
  LinearInterpolate(
   A=@n13
   B=DepthFade(Opacity=@n13, FadeDistance=ScalarParameter() {defaultValue=150, parameterName=Fade_Distance})
   Alpha=ScalarParameter() {parameterName=USE_DEPTH_FADE}
  )
 ]
}
