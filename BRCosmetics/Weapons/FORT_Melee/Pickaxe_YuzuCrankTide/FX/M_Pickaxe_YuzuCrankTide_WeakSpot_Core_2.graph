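# Translucent unlit Niagara sprite material for the core of the YuzuCrankTide pickaxe weak-spot effect.
# A tiled grid texture indexes a curve-atlas LUT, with particle temperature setting its range and tinting an added hot erosion edge.
# Dynamic-parameter erosion sweeps smoothstep bands through a center texture, while FX_Lerp blends edge and trail shapes.
# The trail band, mask, depth fade and particle alpha form opacity; FX_Grid_Strength is declared but unused.
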Material /BRCosmetics/Weapons/FORT_Melee/Pickaxe_YuzuCrankTide/FX/M_Pickaxe_YuzuCrankTide_WeakSpot_Core_2
graph {
 properties {
  bAutomaticallySetUsageInEditor=false
  bUsedWithNiagaraSprites=true
  blendMode=BLEND_Translucent
  shadingModel=MSM_Unlit
  usedShadingModels=Unlit
 }
 nodes [
  n1 = ScalarParameter() {parameterName=LUT_MAX_LOW}
  n2 = ScalarParameter() {parameterName=LUT MAX_HIGH}
  n3 = DynamicParameter() {defaultValue=<rgba 0 1 0 1>, paramNames [FX_Erosion, FX_Temperature, FX_Lerp, FX_Grid_Strength]}
  n4 = Multiply(A=TextureCoordinate(), B=ScalarParameter() {defaultValue=13, parameterName=Grid_Count}) {constB=13}
  n5 = Multiply(
   A=Saturate(None=TextureSampleParameter2D(UVs=@n4) {parameterName=Main_TX, texture=&/T_YuzuCrankTide_Grid}.RGB)
   B=ScalarParameter() {defaultValue=1, parameterName=Grid_Strength}
  )
  n6 = CurveAtlasRowParameter(CurveTime=LinearInterpolate(
    A=ScalarParameter() {defaultValue=0.2, parameterName=LUT MIN}
    B=LinearInterpolate(A=@n1, B=@n2, Alpha=@n3.FX_Temperature)
    Alpha=Saturate(None=@n5)
   )) {
   atlas=&/CA_Pickaxe_YuzuCrankTide
   curve=&/C_Pickaxe_YuzuCrankTide
   parameterName=BaseLUT
  }
  n7 = LinearInterpolate(
   A=VectorParameter() {parameterName=Hot_Edge_Color_Mult_Min, defaultValue=<rgba 1 1 1 1>}.RGB
   B=VectorParameter() {parameterName=Hot_Edge_Color_Mult_Max, defaultValue=<rgba 1 1 1 1>}.RGB
   Alpha=DynamicParameter() {defaultValue=<rgba 0 1 0 1>, paramNames [FX_Erosion, FX_Temperature, FX_Lerp, FX_Grid_Strength]}.FX_Temperature
  )
  n8 = Subtract(
   A=TextureSampleParameter2D() {parameterName=TX_Erosion, texture=&/T_Pickaxe_YuzuCrankTide_Weakhit_Center}.R
   B=LinearInterpolate(
    A=ScalarParameter() {defaultValue=0.2, parameterName=Edge_Width_A}
    B=ScalarParameter() {parameterName=Edge_Width_B}
    Alpha=DynamicParameter() {defaultValue=<rgba 0 1 0 1>, paramNames [FX_Erosion, FX_Temperature, FX_Lerp, FX_Grid_Strength]}.FX_Lerp
   ) {constAlpha=0}
  )
  n9 = ScalarParameter() {defaultValue=0.1, parameterName=Falloff_Erosion}
  n10 = MaterialFunctionCall(
   Input=Add(
    A=ScalarParameter() {parameterName=Progress_Erosion_Offset}
    B=DynamicParameter() {defaultValue=<rgba 0 1 0 1>, paramNames [FX_Erosion, FX_Temperature, FX_Lerp, FX_Grid_Strength]}.FX_Erosion
   )
   Input Low=Constant()
   Input High=Constant() {r=1}
   Target Low=Subtract(B=@n9) {constA=0}
   Target High=Constant() {r=1}
  ) {materialFunction=&RemapValueRange}
  n11 = Add(A=@n10.Result, B=@n9)
  n12 = Multiply(
   A=OneMinus(None=SmoothStep(Min=@n10.Result, Max=@n11, Value=@n8))
   B=LinearInterpolate(
    A=ScalarParameter() {defaultValue=1, parameterName=Edge_Strength_A}
    B=ScalarParameter() {parameterName=Edge_Strength_B}
    Alpha=DynamicParameter() {defaultValue=<rgba 0 1 0 1>, paramNames [FX_Erosion, FX_Temperature, FX_Lerp, FX_Grid_Strength]}.FX_Lerp
   )
  )
  n13 = ParticleColor()
  n14 = Multiply(A=Add(A=@n6, B=Multiply(A=@n7, B=@n12)), B=@n13.RGB)
  Divide(A=@n14, B=ViewProperty() {property=MEVP_PreExposure}.Property)
  MaterialFunctionCall(Emissive=@n14) {materialFunction=&MF_FX_EmissiveAdjustments}
  n15 = TextureSampleParameter2D() {parameterName=TX_Erosion, texture=&T_Pickaxe_YuzuCrankTide_Weakhit_Center}
  n16 = Subtract(
   A=@n15.R
   B=LinearInterpolate(
    A=ScalarParameter() {defaultValue=0.5, parameterName=Trail_Width_A}
    B=ScalarParameter() {parameterName=Trail_Width_B}
    Alpha=DynamicParameter() {defaultValue=<rgba 0 1 0 1>, paramNames [FX_Erosion, FX_Temperature, FX_Lerp, FX_Grid_Strength]}.FX_Lerp
   )
  )
  n17 = Multiply(A=OneMinus(None=SmoothStep(Min=@n10.Result, Max=@n11, Value=@n16)), B=SmoothStep(Min=@n10.Result, Max=@n11, Value=@n15.R))
  n18 = TextureSampleParameter2D() {parameterName=TX_Erosion_Mask, texture=&/T_Pickaxe_YuzuCrankTide_Weakhit_Mask}
  n19 = ScalarParameter() {defaultValue=100, parameterName=DepthFade_Distance}
  Multiply(A=@n13.A, B=Multiply(A=Multiply(A=@n17, B=@n18.R), B=DepthFade(FadeDistance=@n19)))
 ]
}
