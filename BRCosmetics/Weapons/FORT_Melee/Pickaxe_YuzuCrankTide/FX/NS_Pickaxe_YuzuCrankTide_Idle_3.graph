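# Niagara idle glow effect for the YuzuCrankTide pickaxe, with an ambient layer and a reactive layer in red-orange tones.
# Idle_Glow continuously spawns short-lived oriented sprites that quickly fade in, then slowly out.
# Reactive_Glow bursts persistent particles through three sprite and two mesh renderers, with separate glow, tight and bloom colors.
# User.ReactiveOn feeds the reactive alpha each update, and front-end modifiers dim its base and bloom colors outside gameplay.
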NiagaraSystem /BRCosmetics/Weapons/FORT_Melee/Pickaxe_YuzuCrankTide/FX/NS_Pickaxe_YuzuCrankTide_Idle_3
dependencies {usedDataInterfaces [&NiagaraDataInterfaceCurve], usedDynamicInputs [], usedModules [&SystemState]}
emitters {
 Idle_Glow {
  emitterUpdateScript [
   &EmitterState()
   &SpawnBurst_Instantaneous(Use Spawn Probability=false, UseLoopCountLimit=false)
   &SpawnRate(SpawnRate=2, Use Spawn Probability=false)
  ]
  particleSpawnScript [
   &InitializeParticle(
    Color=<rgba 0.9062 0.0943 0.0661 0.5>
    Lifetime=2
    Position Offset=<-6.4 -0.6 -0.5>
    Position Offset Coordinate Space=::Local
    Sprite Rotation Angle=80
    Sprite Rotation Mode=::NewEnumerator0
    Sprite Size Mode=::NewEnumerator3
    Uniform Sprite Size=38
    UsePositionOffset=true
    Write Material Random=true
   )
   &SpriteFacingAndAlignment(Sprite Facing=<-0.15 1 -0.05>, WriteAlignment=true, WriteFacing=true)
  ]
  particleUpdateScript [
   &ParticleState()
   &ScaleColor(
    Scale Alpha=&FloatFromCurve(FloatCurve=DataInterface() {propertyValues {Curve {keys [<at 0 0>,<at 0.1096 1>,<at 1 0>]}}})
    ScaleA=true
    ScaleRGB=true
   )
  ]
  rendererClasses [&NiagaraSpriteRendererProperties]
  renderers [{rendererClass=&NiagaraSpriteRendererProperties}]
 }
 Reactive_Glow {
  emitterUpdateScript [
   &EmitterState()
   &SpawnBurst_Instantaneous(Use Spawn Probability=false, UseLoopCountLimit=false)
  ]
  particleSpawnScript [
   &InitializeParticle(
    Color=&Multiply_LinearColorByFloat(Linear Color=<rgba 1 0 0.0548 1>, Float=&FN_FrontEndFloatModifier(Gameplay=1, Front End=0.7))
    Lifetime=2
    Mesh Scale=<0.74 0.74 2>
    Mesh Scale Mode=::NewEnumerator4
    Position Offset=<-5.4 -0.6 -2.3>
    Position Offset Coordinate Space=::Local
    Sprite Rotation Angle=80.4
    Sprite Rotation Mode=::NewEnumerator0
    Sprite Size Mode=::NewEnumerator3
    Uniform Sprite Size=44
    UsePositionOffset=true
    Write Material Random=true
   )
   &SpriteFacingAndAlignment(Sprite Facing=<-0.15 1 -0.05>, WriteAlignment=true, WriteFacing=true)
   &CameraOffset(Camera Offset Amount=1)
   &InitialMeshOrientation(
    Mesh Axis To Orient=<0 0 1>
    Mesh Orientation Mode=::NewEnumerator2
    Orientation Vector=<-0.15 1 -0.05>
    Rotation=<0 0 0.446>
    Use Rotation Vector=true
   )
   /SetVariables(Particles.GlowColor=&Multiply_LinearColorByFloat(Linear Color=<rgba 0.5 0 0.0177 0.3>, Float=1)) {isSetParameters=true}
   /SetVariables(Particles.TightGlowColor:LinearColor=<rgba 2 0.38 0.0645 1>) {isSetParameters=true}
   /SetVariables(Particles.TightGlowColorScale:Vector3f=&Multiply_Vector(A=$Particles.Scale, B=<1 1 0.5>)) {isSetParameters=true}
  ]
  particleUpdateScript [
   &ParticleState(Kill Particles When Lifetime Has Elapsed=false, Let Infinitely Lived Particles Die When Emitter Deactivates=true)
   &UpdateMeshOrientation(Rotation Vector=<0 0 1>) {enabled=false}
   /SetVariables(Particles.ReactiveAlpha:NiagaraFloat=$User.ReactiveOn) {isSetParameters=true}
   /SetVariables(
    Particles.AmbientSize:Vector2f=&Multiply_Vector2D(A=$Particles.SpriteSize, B=&Vector2DFromFloat(Value=1.1))
    Particles.BloomColor:LinearColor=&Multiply_LinearColorByFloat(Linear Color=<rgba 15 2.5394 0 1>, Float=&FN_FrontEndFloatModifier(Gameplay=1, Front End=0.6))
    Particles.BloomSize:Vector2f=&Multiply_Vector2D(A=$Particles.SpriteSize, B=&Vector2DFromFloat(Value=1.05))
    Particles.TightColor:LinearColor=<rgba 50 7.6515 1.8229 1>
   ) {isSetParameters=true}
  ]
  rendererClasses [&NiagaraSpriteRendererProperties, &NiagaraMeshRendererProperties]
  renderers [
   {rendererClass=&NiagaraSpriteRendererProperties}
   {rendererClass=&NiagaraSpriteRendererProperties, rendererIndex=1}
   {rendererClass=&NiagaraSpriteRendererProperties, rendererIndex=2}
   {rendererClass=&NiagaraMeshRendererProperties, rendererIndex=3}
   {rendererClass=&NiagaraMeshRendererProperties, rendererIndex=4}
  ]
 }
}
summary {
 systemName=NS_Pickaxe_YuzuCrankTide_Idle_3
 emitters [
  {emitterName=Idle_Glow, rendererClasses [&NiagaraSpriteRendererProperties]}
  {emitterName=Reactive_Glow, rendererClasses [&NiagaraSpriteRendererProperties, &NiagaraMeshRendererProperties]}
 ]
 userVariables [
  User.bIsFrontEnd {type=NiagaraBool}
  User.bIsFrontEndPreview {type=NiagaraBool}
  User.ReactiveBool {type=NiagaraBool}
  User.ReactiveOn {type=NiagaraFloat, defaultValue=1, description=""}
  User.ReactivityOn {type=NiagaraFloat, defaultValue=1, description=""}
  User.Timestamp {type=NiagaraFloat, defaultValue=1, description=""}
 ]
}
