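# Niagara weapon trail for the YuzuCrankTide pickaxe, drawing swing ribbons from bone or socket pairs on the skeletal mesh.
# A single Trail emitter uses FN AnimTrail modules, selecting bone indices and width by swing section index, scaled by user width and lifetime multipliers.
# Time-of-day scalars tune alpha and glow, while color curves animate two edges and a front glow feeding four ribbon renderers.
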NiagaraSystem /BRCosmetics/Weapons/FORT_Melee/Pickaxe_YuzuCrankTide/FX/NS_Pickaxe_YuzuCrankTide_Trail
dependencies {
 usedDataInterfaces [
  &NiagaraDataInterfaceArrayInt32
  &NiagaraDataInterfaceCurve
  &NiagaraDataInterfaceArrayFloat
  &NiagaraDataInterfaceColorCurve
  &NiagaraDataInterfaceSocketReader
 ]
 usedDynamicInputs []
 usedModules [
  &self:SystemSpawnScript.SystemScriptSource.SystemScriptGraph.NiagaraNodeAssignment_0.SetVariables_#1
  &SystemState
 ]
}
emitters {Trail {
  emitterSpawnScript [&FN_RibbonIDOffsetUtility()]
  emitterUpdateScript [
   &EmitterState()
   &FN_SpawnAnimTrail()
  ]
  particleSpawnScript [
   &FN_InitializeAnimTrailBoneIndecies(
    First Bone or Socket Index:NiagaraInt32=&SelectIntFromArray(Direct Array Index=$User.SwingSectionIndex, Int Selection Array=DataInterface() {propertyValues {IntData [1,1,1,1]}})
    Second Bone or Socket Index:NiagaraInt32=&SelectIntFromArray(Direct Array Index=$User.SwingSectionIndex, Int Selection Array=DataInterface() {propertyValues {IntData [2,0,2,0]}})
   ) {module=FN_InitializeAnimTrailBoneIndecies002}
   &FN_InitializeAnimTrail(
    AnimTrailWidth=&Multiply_Float(
     A=&SelectFloatFromArray(Direct Array Index=$User.SwingSectionIndex, Float Selection Array=DataInterface() {propertyValues {FloatData [0.55,0.6,0.55,0.55]}})
     B=$User.WidthMult
    )
    Color=&MultiplyLinearColor(
     Linear Color=<rgba 1 0 0 1>
     Scale Factor=&MakeVector4(
      X=1
      Y=1
      Z=1
      W=&FloatFromCurve(
       FloatCurve=DataInterface() {propertyValues {Curve {keys [<at 0 1>,<at 0.1 1>,<at 0.5636 1>,<at 0.656 1>,<at 0.9 0>]}}}
       Scale Curve=&FNTimeOfDayScalar(Night Scale=4, Morning Scale=3, Day Scale=3, Evening Scale=3)
      )
     )
    )
    Lifetime=&Multiply_Float(A=1, B=$User.TrailLifeTimeMult)
   )
  ]
  particleUpdateScript [
   &SolveForcesAndVelocity(Clamp Velocity=false, Limit Acceleration=false)
   &FN_UpdateAnimTrail() {module=FN_UpdateAnimTrail001}
   &ScaleColor(
    Scale Alpha=&FloatFromCurve(
     FloatCurve=DataInterface() {propertyValues {Curve {keys [
         {arriveTangent=3.3333, interpMode=RCIM_Cubic, time=0.01, value=1}
         {arriveTangent=-5.2632, interpMode=RCIM_Cubic, leaveTangent=-5.2632, time=0.15, value=0.5}
         <at 0.2 0 RCIM_Cubic>
        ]}}}
     Scale Curve=&FNTimeOfDayScalar(Night Scale=0.2, Morning Scale=0.3, Day Scale=0.3, Evening Scale=0.3)
    )
    ScaleA=true
    ScaleRGB=false
   )
   /SetVariables(
    Particles.EdgeColorBot:LinearColor=&ColorFromCurve(ColorCurve=DataInterface() {propertyValues {
       AlphaCurve {keys [<at 0 0>,<at 0.0391 0.5 RCIM_Cubic>,<at 0.1273 0 RCIM_Cubic>]}
       BlueCurve {keys [<at 0 0.013 RCIM_Cubic>,<at 0.2884 0.013 RCIM_Cubic>]}
       GreenCurve {keys [<at 0 0 RCIM_Cubic>,<at 0.2884 0 RCIM_Cubic>]}
       RedCurve {keys [<at 0 0.469 RCIM_Cubic>,<at 0.2884 0.469 RCIM_Cubic>]}
      }})
    Particles.EdgeWidthBot:NiagaraFloat=&Multiply_Float(A=19, B=$User.WidthMult)
   ) {isSetParameters=true}
   /SetVariables(
    Particles.EdgeColor:LinearColor=&ColorFromCurve(ColorCurve=DataInterface() {propertyValues {
       AlphaCurve {keys [<at 0.0378 2 RCIM_Cubic>,<at 0.204 0 RCIM_Cubic>]}
       BlueCurve {keys [<at 0 0.2865 RCIM_Cubic>,<at 0.4545 0.1951 RCIM_Cubic>]}
       GreenCurve {keys [<at 0 0.9441 RCIM_Cubic>,<at 0.4545 0.1092 RCIM_Cubic>]}
       RedCurve {keys [<at 0 5 RCIM_Cubic>,<at 0.4545 5 RCIM_Cubic>]}
      }})
    Particles.EdgeWidth:NiagaraFloat=&Multiply_Float(A=19.1, B=$User.WidthMult)
   ) {isSetParameters=true}
   /SetVariables(
    Particles.FrontGlowColor:LinearColor=&Multiply_LinearColorByFloat(
     Linear Color=&ColorFromCurve(ColorCurve=DataInterface() {propertyValues {
        AlphaCurve {keys [<at 0 0.15 RCIM_Cubic>,<at 0.05 0 RCIM_Cubic>]}
        BlueCurve {keys [<at 0 1.6406 RCIM_Cubic>,<at 0.0001 0 RCIM_Cubic>]}
        GreenCurve {keys [<at 0 4.5183 RCIM_Cubic>,<at 0.0001 0 RCIM_Cubic>]}
        RedCurve {keys [<at 0 15 RCIM_Cubic>,<at 0.001 7.2131 RCIM_Cubic>]}
       }})
     Float=&FNTimeOfDayScalar(Night Scale=1, Morning Scale=0.5, Evening Scale=0.5)
    )
    Particles.FrontGlowWidth:NiagaraFloat=&Multiply_Float(A=13.5, B=$User.WidthMult)
   ) {isSetParameters=true}
  ]
  rendererClasses [&NiagaraRibbonRendererProperties]
  renderers [
   {rendererClass=&NiagaraRibbonRendererProperties}
   {rendererClass=&NiagaraRibbonRendererProperties, rendererIndex=1}
   {rendererClass=&NiagaraRibbonRendererProperties, rendererIndex=2}
   {rendererClass=&NiagaraRibbonRendererProperties, rendererIndex=3}
  ]
 }}
summary {
 systemName=NS_Pickaxe_YuzuCrankTide_Trail
 emitters [{emitterName=Trail, rendererClasses [&NiagaraRibbonRendererProperties]}]
 userVariables [
  User.bIsFrontEnd {type=NiagaraBool}
  User.bIsFrontEndPreview {type=NiagaraBool}
  User.RibbonIDOffset {type=NiagaraInt32}
  User.SkeletalMesh {
   type=NiagaraDataInterfaceSkeletalMesh
   defaultValue {
    struct=NiagaraExt_VariableValue_DataInterface
    value {dataInterface=&self:NiagaraDataInterfaceSkeletalMesh_0}
   }
  }
  User.SwingSectionIndex {type=NiagaraInt32}
  User.TrailLifeTimeMult {type=NiagaraFloat, defaultValue=0.25, description=""}
  User.WidthMult {type=NiagaraFloat, defaultValue=2.6, description=""}
 ]
}
