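# Parent material for the YuzuCrankTide pickaxe, adding a reactive emissive edge glow between the CharacterParent_2023 PreFX and PostFX functions.
# A mask texture's green channel tints emissive toward GlowEdgeColor, scaled by reactivity and gated by ReactivityOn.
# Timestamp parameters build a normalized deploy timer and two overlapping kill fade-in/out envelopes, combined with Max.
# Pre-skinned local position provides a deploy sweep mask and angular flow coordinates, alongside panned radial-noise and sine-pulse branches.
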Material /BRCosmetics/Weapons/FORT_Melee/Pickaxe_YuzuCrankTide/Materials/M_YuzuCrankTide_Pickaxe_Parent
graph {
 properties {
  bUseMaterialAttributes=true
  bUsedWithMeshDeformer=true
  bUsedWithMorphTargets=true
  bUsedWithNanite=true
  bUsedWithNiagaraMeshParticles=true
  bUsedWithSkeletalMesh=true
  pixelDepthOffsetMode=PDOM_Legacy
  shadingModel=MSM_FromMaterialExpression
  usedShadingModels=Default Lit | Subsurface | Clear Coat | Subsurface Profile | Cloth
 }
 nodes [
  Frac(None=Multiply(A=Time()))
  n1 = NamedRerouteUsage()
  n2 = Multiply(
   A=MaterialFunctionCall(Vector or UVs=Multiply(A=@n1) {constB=5}) {materialFunction=&VectorToRadialValue}.Radial Coordinates
   B=Constant2Vector() {g=1, r=2}
  )
  LinearInterpolate(Alpha=TextureSample(UVs=Panner(Coordinate=@n2) {speedX=1}) {samplerType=SAMPLERTYPE_LinearGrayscale, texture=/Game/Global/Textures/Default/Noise/EqualizerComponents_v2_Noise}.R) {constA=-0.2, constB=2}
  n3 = ScalarParameter() {parameterName=ReactivityGlowIntensity}
  n4 = NamedRerouteUsage()
  LinearInterpolate(B=@n3, Alpha=@n4) {constA=1}
  n5 = MaterialFunctionCall() {materialFunction=&MF_CharacterParent_2023_PreFX}
  MaterialFunctionCall(Material=NamedRerouteUsage(), M=@n5.M, Pre Skinned Local Position=@n5.M, Distance Blend=@n5.M, Fresnel=@n5.M) {materialFunction=&MF_CharacterParent_2023_PostFX}
  n6 = TextureSample() {
   samplerType=SAMPLERTYPE_LinearGrayscale
   texture=&/T_YukuCrankTide_EdgeGlow_Mask
  }
  n7 = Multiply(A=NamedRerouteUsage())
  n8 = Multiply(A=Time()) {constB=0.5}
  n9 = ConstantBiasScale(None=Sine(None=Add(A=@n7, B=@n8)))
  Multiply(A=@n6.G, B=Add(A=@n6.G, B=Multiply(A=@n9, B=@n9)))
  NamedRerouteDeclaration(None=@n5.Result) {name=InMaterial, nodeColor=<rgba 0.5469 0 1 1>}
  n10 = GetMaterialAttributes(None=NamedRerouteUsage()) {attributeGetTypes [#1]}
  NamedRerouteDeclaration(None=SetMaterialAttributes(MaterialAttributes=@n10.MaterialAttributes, Emissive Color=NamedRerouteUsage()) {attributeSetTypes [#1]}) {name=OutMaterial, nodeColor=<rgba 0 1 0.2266 1>}
  NamedRerouteDeclaration(None=@n10.Emissive Color) {name=InEmissive, nodeColor=<rgba 1 0 0.0937 1>}
  n11 = NamedRerouteUsage()
  n12 = VectorParameter() {parameterName=GlowEdgeColor, defaultValue=<rgba 25 0.1964 0 1>}
  n13 = Multiply(A=Saturate(None=Power(Base=@n6.G)), B=Max(A=@n4, B=@n3))
  NamedRerouteDeclaration(None=LinearInterpolate(A=@n11, B=LinearInterpolate(A=@n11, B=@n12.RGB, Alpha=@n13), Alpha=ScalarParameter() {defaultValue=1, parameterName=ReactivityOn})) {name=OutEmissive, nodeColor=<rgba 0 0.3437 1 1>}
  n14 = Add(A=LocalPosition() {localOrigin=InstancePreSkinning}.XYZ, B=Constant3Vector() {constant=<rgba 60 0 0 0>})
  n15 = VertexInterpolator(VS=Multiply(A=ComponentMask(None=@n14) {r=true}) {constB=0.01})
  n16 = NamedRerouteUsage()
  n17 = Saturate(None=Add(A=Subtract(A=@n15.PS), B=SmoothStep(Value=@n16)))
  n18 = Constant()
  NamedRerouteDeclaration(None=OneMinus(None=SphereMask(A=@n17, B=@n18) {attenuationRadius=0.2, hardnessPercent=50})) {name=DeployMask, nodeColor=<rgba 1 0 0.75 1>}
  n19 = Subtract(
   A=Time()
   B=Add(A=ScalarParameter() {defaultValue=-999999, parameterName=Timestamp}, B=ScalarParameter() {defaultValue=-0.25, parameterName=DeployDelay}) {constB=0.25}
  )
  n20 = ScalarParameter() {defaultValue=1.3, parameterName=DeployAnimationLength}
  NamedRerouteDeclaration(None=Saturate(None=Divide(A=@n19, B=@n20))) {name=DeployTimestamp, nodeColor=<rgba 0.8047 1 0 1>}
  n21 = Time()
  n22 = ScalarParameter() {defaultValue=-100, parameterName=Timestamp_0}
  n23 = ScalarParameter() {defaultValue=1.5, parameterName=KillTransitionLength}
  n24 = Saturate(None=Divide(A=Subtract(A=@n21, B=@n22), B=@n23) {desc=InLength})
  n25 = ScalarParameter() {defaultValue=2, parameterName=KillAnimationLength}
  n26 = Subtract(A=@n21, B=Add(A=@n22, B=Add(A=@n23, B=@n25) {constB=2, desc=eXTRAdURATION}))
  n27 = OneMinus(None=Saturate(None=Divide(A=@n26, B=@n23) {constB=1.5, desc=OutLength}))
  n28 = Time()
  n29 = ScalarParameter() {defaultValue=-100, parameterName=Timestamp_1}
  n30 = ScalarParameter() {defaultValue=1.5, parameterName=KillTransitionLength}
  n31 = Saturate(None=Divide(A=Subtract(A=@n28, B=@n29), B=@n30) {desc=InLength})
  n32 = ScalarParameter() {defaultValue=2, parameterName=KillAnimationLength}
  n33 = Subtract(A=@n28, B=Add(A=@n29, B=Add(A=@n30, B=@n32) {constB=2, desc=eXTRAdURATION}))
  n34 = OneMinus(None=Saturate(None=Divide(A=@n33, B=@n30) {constB=1.5, desc=OutLength}))
  NamedRerouteDeclaration(None=Max(A=Multiply(A=@n24, B=@n27), B=Multiply(A=@n31, B=@n34))) {name=KillTimestamp, nodeColor=<rgba 0.8047 1 0 1>}
  n35 = Time()
  Saturate(None=Sine(None=Multiply(A=@n35) {constB=0.3}))
  n36 = Add(A=LocalPosition() {localOrigin=InstancePreSkinning}.XYZ, B=Constant3Vector() {constant=<rgba 7 0 -2.7154 0>})
  n37 = RotateAboutAxis(
   NormalizedRotationAxis=Constant3Vector() {constant=<rgba 0 1 0 0>}
   RotationAngle=ScalarParameter() {defaultValue=0.25}
   PivotPoint=Constant()
   Position=@n36
  )
  n38 = Multiply(A=ComponentMask(None=Add(A=@n36, B=@n37)) {b=true, r=true}) {constB=0.01}
  n39 = MaterialFunctionCall(Float2=NamedRerouteDeclaration(None=VertexInterpolator(VS=@n38).PS) {name=PreSkinnedCoords, nodeColor=<rgba 0 1 0.6953 1>}) {materialFunction=&BreakOutFloat2Components}
  n40 = Divide(
   A=Arctangent2Fast(Y=@n39.R, X=@n39.R)
   B=MaterialFunctionCall() {materialFunction=&Pi}.Result
  )
  SphereMask(
   A=NamedRerouteDeclaration(None=Abs(None=@n40)) {name=FlowNoise, nodeColor=<rgba 0 0.9766 1 1>}
   B=Constant() {r=1}
   Radius=LinearInterpolate(Alpha=@n16)
  ) {attenuationRadius=1, hardnessPercent=50}
 ]
}
